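# Physics asset for the CyberMitt backpack cosmetic, set up for two dangling cords, A and B.
# Capsule bodies cover the pack base and each cord's master bone plus three segments.
# Each master is locked to the base in swing and twist, anchoring a chain whose swing limits widen toward the tip, with twist limits at zero.
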PhysicsAsset /BRCosmetics/Accessories/FORT_Backpacks/Backpack_M_MED_CyberMitt/Meshes/M_MED_CyberMitt_Pack_Physics
bodies {
 base [{shapeType=Capsule, length=37.6101, radius=3.4351, center=<3.5034 -0.7454 -1.1226>, rotation=<rot 0 90.9015 -89.4197>}]
 dyn_cordA_01 [{shapeType=Capsule, length=2, radius=0.5, shapeName=dyn_cordA_01_capsule, center=<0.8184 0 0.2275>, rotation=<rot -90 0 0>}]
 dyn_cordA_02 [{shapeType=Capsule, length=2.5, radius=1, shapeName=dyn_cordA_02_capsule, center=<1.0514 -0.0315 0.0221>, rotation=<rot -90 -0.0002 0>}]
 dyn_cordA_03 [{shapeType=Capsule, length=3.5, radius=1, shapeName=dyn_cordA_03_capsule, center=<3.0969 -0.0327 -0.0498>, rotation=<rot -90 0 0>}]
 dyn_cordA_MSTR [{shapeType=Capsule, length=0.5, radius=0.5, shapeName=dyn_cordA_MSTR_capsule, center=<-0.4708 0 -0.4571>}]
 dyn_cordB_01 [{shapeType=Capsule, length=2.5, radius=0.5, shapeName=dyn_cordB_01_capsule, center=<1.0282 0.1063 0.2325>, rotation=<rot -90 -0.0005 0>}]
 dyn_cordB_02 [{shapeType=Capsule, length=2.5, radius=1, shapeName=dyn_cordB_02_capsule, center=<1.3473 -0.0005 0.002>, rotation=<rot -90 -0.001 0>}]
 dyn_cordB_03 [{shapeType=Capsule, length=3.5, radius=1, shapeName=dyn_cordB_03_capsule, center=<4.1788 -0.1249 -0.1653>, rotation=<rot -90 -2.1593 0>}]
 dyn_cordB_MSTR [{shapeType=Capsule, length=0.5, radius=0.5, shapeName=dyn_cordB_MSTR_capsule}]
}
constraints [
 {bone1Name=dyn_cordA_02, bone2Name=dyn_cordA_01, swing1LimitDegrees=45, swing2LimitDegrees=35, twistLimitDegrees=0}
 {bone1Name=dyn_cordA_03, bone2Name=dyn_cordA_02, swing1LimitDegrees=60, swing2LimitDegrees=45, twistLimitDegrees=0}
 {bone1Name=dyn_cordB_01, bone2Name=dyn_cordB_MSTR, swing1LimitDegrees=35, swing2LimitDegrees=25, twistLimitDegrees=0}
 {bone1Name=dyn_cordB_02, bone2Name=dyn_cordB_01, swing1LimitDegrees=45, swing2LimitDegrees=35, twistLimitDegrees=0}
 {bone1Name=dyn_cordB_03, bone2Name=dyn_cordB_02, swing1LimitDegrees=60, swing2LimitDegrees=45, twistLimitDegrees=0}
 {
  bone1Name=dyn_cordA_MSTR
  bone2Name=base
  swing1LimitDegrees=45
  swing1Motion=Locked
  swing2LimitDegrees=45
  swing2Motion=Locked
  twistLimitDegrees=45
  twistMotion=Locked
 }
 {
  bone1Name=dyn_cordB_MSTR
  bone2Name=base
  swing1LimitDegrees=45
  swing1Motion=Locked
  swing2LimitDegrees=45
  swing2Motion=Locked
  twistLimitDegrees=45
  twistMotion=Locked
 }
 {bone1Name=dyn_cordA_01, bone2Name=dyn_cordA_MSTR, swing1LimitDegrees=35, swing2LimitDegrees=25, twistLimitDegrees=0}
]
