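# Animation Blueprint for the CyberMitt backpack, animating its straps and cords and toggling a hideable material element.
# With rigid body enabled, spine-driven root offsets and a skydive base offset feed strap AnimDynamics and a RigidBody node.
# Otherwise, AnimDynamics cord chains run on the input pose without those offsets.
# Weapon, customization and skydiving events set "Hide Element 01" when a CyberMitt-tagged pickaxe is equipped outside skydiving; the swimming alpha is unused.
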AnimBlueprint /BRCosmetics/Accessories/FORT_Backpacks/Backpack_M_MED_CyberMitt/Meshes/M_MED_CyberMitt_Pack_AnimBP
target_skeleton=&/M_MED_CyberMitt_Pack_Skeleton
graphs [
 AnimGraph {nodes [
   n1 = "Transform (Modify) Bone - Bone: root"(
    ComponentPose=Local To Component(LocalPose=Use cached pose 'Base_pose'())
    Translation=Get backpack_translation()
    Rotation=Get backpack_rotate()
   ) {
    rotationMode=BMM_Additive
    rotationSpace=BCS_BoneSpace
    scaleMode=BMM_Additive
    scaleSpace=BCS_BoneSpace
    translationMode=BMM_Additive
    translationSpace=BCS_BoneSpace
    alphaBoolBlend {blendOption=Linear}
    boneToModify {boneName=root}
   }
   n2 = "Transform (Modify) Bone - Bone: base"(ComponentPose=@n1, Alpha=Get Alpha_skydiveMode() -> (Alpha_skydiveMode:double), Translation="-10,-3,5", Rotation="0,0,0", Scale="1,1,1") {
    rotationMode=BMM_Additive
    rotationSpace=BCS_BoneSpace
    translationMode=BMM_Additive
    translationSpace=BCS_BoneSpace
    alphaBoolBlend {blendOption=Linear}
    boneToModify {boneName=base}
    translation=<-10 -3 5>
   }
   n3 = Get Gravity Override()
   n4 = Anim Dynamics - Bone: dyn_strap_A!(ComponentPose=@n2, Alpha=1, GravityOverride:Vector=@n3, ExternalForce="0,0,0") -> (Pose) {
    angularDampingOverride=1
    angularSpringConstant=10
    bOverrideAngularDamping=true
    bUseGravityOverride=true
    gravityScale=0.35
    simulationSpace=RootRelative
    alphaBoolBlend {blendOption=Linear}
    boundBone {boneName=dyn_strap_A}
    componentAppliedLinearAccClamp=<7000 7000 7000>
    componentLinearAccScale=<0.75 0.75 0.75>
    componentLinearVelScale=<0.75 0.75 0.75>
    gravityOverride {z=-980}
    physicsBodyDefinitions [{
      boundBone {boneName=dyn_strap_A}
      boxExtents=<10 10 10>
      constraintSetup {angularLimitsMax=<30 30 30>, angularLimitsMin=<-30 -30 -30>, angularTarget=<1 0 0>}
      localJointOffset=<1 0 0>
     }]
   }
   n5 = Anim Dynamics - Bone: dyn_strap_B!(ComponentPose=@n4, Alpha=1, GravityOverride:Vector=@n3, ExternalForce="0,0,0") -> (Pose) {
    angularDampingOverride=1
    angularSpringConstant=10
    bOverrideAngularDamping=true
    bUseGravityOverride=true
    gravityScale=0.35
    simulationSpace=RootRelative
    alphaBoolBlend {blendOption=Linear}
    boundBone {boneName=dyn_strap_B}
    componentAppliedLinearAccClamp=<7000 7000 7000>
    componentLinearAccScale=<0.75 0.75 0.75>
    componentLinearVelScale=<0.75 0.75 0.75>
    gravityOverride {z=-980}
    physicsBodyDefinitions [{
      boundBone {boneName=dyn_strap_B}
      boxExtents=<10 10 10>
      constraintSetup {angularLimitsMax=<30 30 30>, angularLimitsMin=<-30 -30 -30>, angularTarget=<1 0 0>}
      localJointOffset=<1 0 0>
     }]
   }
   n6 = RigidBody!(ComponentPose=@n5, Alpha, OverrideWorldGravity:Vector=@n3, ExternalForce) -> (Pose) {
    bClampLinearTranslationLimitToRefPose=true
    bOverrideWorldGravity=true
    overridePhysicsAsset=&/M_MED_CyberMitt_Pack_Physics
    simulationSpace=BaseBoneSpace
    alphaBoolBlend {blendOption=Linear}
    baseBoneRef {boneName=pack}
    componentAppliedLinearAccClamp=<7500 7500 7500>
    componentLinearAccScale=<0.65 0.65 0.65>
    componentLinearVelScale=<0.65 0.65 0.65>
    overrideWorldGravity {z=-980}
   }
   n7 = Anim Dynamics - Bone: dyn_cordA_01!(ComponentPose=Local To Component(LocalPose=Use cached pose 'Base_pose'()), Alpha=1, GravityOverride:Vector="0,0,-980", ExternalForce="0,0,0") -> (Pose) {
    angularDampingOverride=1
    angularSpringConstant=10000
    bAngularSpring=true
    bOverrideAngularDamping=true
    bUseGravityOverride=true
    gravityScale=0.35
    simulationSpace=RootRelative
    alphaBoolBlend {blendOption=Linear}
    boundBone {boneName=dyn_cordA_01}
    componentLinearAccScale=<1 1 1>
    componentLinearVelScale=<1 1 1>
    gravityOverride {z=-980}
    physicsBodyDefinitions [{
      boundBone {boneName=dyn_cordA_01}
      boxExtents=<20 30 30>
      constraintSetup {angularLimitsMax=<45 45 45>, angularLimitsMin=<-45 -45 -45>, angularTarget=<1 0 0>}
      localJointOffset=<4 0 0>
     }]
   }
   n8 = Anim Dynamics - Bone: dyn_cordA_02!(ComponentPose=@n7, Alpha=1, GravityOverride:Vector="0,0,-980", ExternalForce="0,0,0") -> (Pose) {
    angularDampingOverride=1
    angularSpringConstant=10000
    bAngularSpring=true
    bOverrideAngularDamping=true
    bUseGravityOverride=true
    gravityScale=0.35
    simulationSpace=RootRelative
    alphaBoolBlend {blendOption=Linear}
    boundBone {boneName=dyn_cordA_02}
    componentLinearAccScale=<1 1 1>
    componentLinearVelScale=<1 1 1>
    gravityOverride {z=-980}
    physicsBodyDefinitions [{
      boundBone {boneName=dyn_cordA_02}
      boxExtents=<20 30 30>
      constraintSetup {angularLimitsMax=<35 35 35>, angularLimitsMin=<-35 -35 -35>, angularTarget=<1 0 0>}
      localJointOffset=<4 0 0>
     }]
   }
   n9 = Anim Dynamics - Bone: dyn_cordA_03!(ComponentPose=@n8, Alpha=1, GravityOverride:Vector="0,0,-980", ExternalForce="0,0,0") -> (Pose) {
    angularDampingOverride=1
    angularSpringConstant=10000
    bAngularSpring=true
    bOverrideAngularDamping=true
    bUseGravityOverride=true
    gravityScale=0.35
    simulationSpace=RootRelative
    alphaBoolBlend {blendOption=Linear}
    boundBone {boneName=dyn_cordA_03}
    componentLinearAccScale=<1 1 1>
    componentLinearVelScale=<1 1 1>
    gravityOverride {z=-980}
    physicsBodyDefinitions [{
      boundBone {boneName=dyn_cordA_03}
      boxExtents=<20 30 30>
      constraintSetup {angularLimitsMax=<25 25 25>, angularLimitsMin=<-25 -25 -25>, angularTarget=<1 0 0>}
      localJointOffset=<4 0 0>
     }]
   }
   n10 = Anim Dynamics - Bone: dyn_cordB_01!(ComponentPose=@n9, Alpha=1, GravityOverride:Vector="0,0,-980", ExternalForce="0,0,0") -> (Pose) {
    angularDampingOverride=1
    angularSpringConstant=10000
    bAngularSpring=true
    bOverrideAngularDamping=true
    bUseGravityOverride=true
    gravityScale=0.35
    simulationSpace=RootRelative
    alphaBoolBlend {blendOption=Linear}
    boundBone {boneName=dyn_cordB_01}
    componentLinearAccScale=<1 1 1>
    componentLinearVelScale=<1 1 1>
    gravityOverride {z=-980}
    physicsBodyDefinitions [{
      boundBone {boneName=dyn_cordB_01}
      boxExtents=<20 30 30>
      constraintSetup {angularLimitsMax=<45 45 45>, angularLimitsMin=<-45 -45 -45>, angularTarget=<1 0 0>}
      localJointOffset=<4 0 0>
     }]
   }
   n11 = Anim Dynamics - Bone: dyn_cordB_02!(ComponentPose=@n10, Alpha=1, GravityOverride:Vector="0,0,-980", ExternalForce="0,0,0") -> (Pose) {
    angularDampingOverride=1
    angularSpringConstant=10000
    bAngularSpring=true
    bOverrideAngularDamping=true
    bUseGravityOverride=true
    gravityScale=0.35
    simulationSpace=RootRelative
    alphaBoolBlend {blendOption=Linear}
    boundBone {boneName=dyn_cordB_02}
    componentLinearAccScale=<1 1 1>
    componentLinearVelScale=<1 1 1>
    gravityOverride {z=-980}
    physicsBodyDefinitions [{
      boundBone {boneName=dyn_cordB_02}
      boxExtents=<20 30 30>
      constraintSetup {angularLimitsMax=<35 35 35>, angularLimitsMin=<-35 -35 -35>, angularTarget=<1 0 0>}
      localJointOffset=<4 0 0>
     }]
   }
   n12 = Anim Dynamics - Bone: dyn_cordB_03!(ComponentPose=@n11, Alpha=1, GravityOverride:Vector="0,0,-980", ExternalForce="0,0,0") -> (Pose) {
    angularDampingOverride=1
    angularSpringConstant=10000
    bAngularSpring=true
    bOverrideAngularDamping=true
    bUseGravityOverride=true
    gravityScale=0.35
    simulationSpace=RootRelative
    alphaBoolBlend {blendOption=Linear}
    boundBone {boneName=dyn_cordB_03}
    componentLinearAccScale=<1 1 1>
    componentLinearVelScale=<1 1 1>
    gravityOverride {z=-980}
    physicsBodyDefinitions [{
      boundBone {boneName=dyn_cordB_03}
      boxExtents=<20 30 30>
      constraintSetup {angularLimitsMax=<25 25 25>, angularLimitsMin=<-25 -25 -25>, angularTarget=<1 0 0>}
      localJointOffset=<4 0 0>
     }]
   }
   Output Pose(Result=Blend Poses by bool(BlendPose_0=Component To Local(ComponentPose=@n6), BlendPose_1=Component To Local(ComponentPose=@n12), bActiveValue=Get bRigidBodyEnabled()) {blendType=Linear}) {layerGroup=None}
   Save cached pose 'Base_pose'(Pose=Input Pose()) {cacheName=Base_pose}
   Get Alpha_swimmingMode() -> (Alpha_swimmingMode:double)
  ]}
 EventGraph {nodes [
   Event BlueprintUpdateAnimation()
   > n1 = Sequence!(execute) -> (then_0, then_1, then_2:Exec, then_3:Exec)
   n2 = GetOwnerSkeletalMeshComponent()
   n3 = BreakRotator(InRot=BreakTransform(InTransform=GetDeltaTransformFromRefPose(self=@n2, BoneName=spine_05, BaseName=spine_01)).Rotation)
   > Set backpack_rotate(backpack_rotate=MakeRotator(Roll=MapRangeClamped(Value=@n3.Yaw, InRangeA=0, InRangeB=-35, OutRangeA=4, OutRangeB=30), Pitch=1, Yaw=-3))
   > Set backpack_translation(backpack_translation=MakeVector(Y=MapRangeClamped(Value=@n3.Yaw, InRangeA=10, InRangeB=-40, OutRangeB=-5)))
   n4 = Branch(
    execute=@n1.then_1
    Condition=OR Boolean!(A=Get bIsSkydiveDiveMode(), B=Get bIsSkydiveFloating(), C:bool=Get bIsParachuteOpened()) -> (ReturnValue)
   )
   > Set Alpha_skydiveMode(Alpha_skydiveMode:double=1) -> (Output_Get:double)
   Set Alpha_skydiveMode(execute=@n4.else, Alpha_skydiveMode=0)
   n5 = Branch(execute=@n1.then_2, Condition=Get bIsSurfaceSwimming())
   > Set Alpha_swimmingMode(Alpha_swimmingMode:double=1) -> (Output_Get:double)
   Set Alpha_swimmingMode(execute=@n5.else, Alpha_swimmingMode=0)
   n6 = Gravity Override State Function DEPRECATED(
    execute=@n1.then_3
    CharacterPartAnimInstance=Self-Reference()
    GameTimeInSeconds=GetGameTimeInSeconds()
    GravityOverrideParams=Get Gravity Override Params()
   )
   > Set Gravity Override(Gravity Override=@n6)
   n7 = Event OnWeaponChanged()
   n8 = GetPartSkeletalMeshComponent()
   n9 = Get Dynamic Material Instance(execute=[@n7.then, @n16.then], SkeletalMeshComponent=@n8)
   > n10 = Get Dynamic Material Instance(SkeletalMeshComponent=@n8, ElementIndex=1)
   n11 = PickaxeItemDefinitionHasAnyMetaTag(Tags=Make Array("[0]"=(TagName="Cosmetics.MetaTags.CyberMitt")), bRequiresEquipped=true)
   > n12 = Branch(Condition=@n11)
   > n13 = Branch(Condition=Get bIsSkyDiving())
   SetScalarParameterValue(execute=[@n14.then, @n15.else, @n13.then, @n12.else], self=@n9, ParameterName=Hide Element 01, Value=0)
   > SetScalarParameterValue(self=@n10, ParameterName=Hide Element 01, Value=0)
   SetScalarParameterValue(execute=[@n15.then, @n13.else], self=@n9, ParameterName=Hide Element 01, Value=1)
   > SetScalarParameterValue(self=@n10, ParameterName=Hide Element 01, Value=1)
   n14 = Event OnBeginSkydiving()
   Event OnEndSkydiving()
   > n15 = Branch(Condition=@n11)
   n16 = Event OnCharacterCustomizationLoaded()
   TryGetPawnOwner()
  ]}
]
variables [
 Alpha_skydiveMode {type=number}
 Alpha_swimmingMode {type=number}
 backpack_rotate {type=Rotator}
 backpack_translation {type=Vector}
 Gravity Override {type=Vector}
 Gravity Override Params {type=GravityOverrideParamsStruct}
]
